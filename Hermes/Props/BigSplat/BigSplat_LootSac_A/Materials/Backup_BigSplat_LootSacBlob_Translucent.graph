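# Backup translucent lit material for the BigSplat loot sac blob, faking a liquid-filled glass container.
# A virtual plane and local-height masks shape the liquid, filled with scrolling, noise-distorted texture layers and time-of-day colors.
# Fresnel drives opacity; skylight reflections and screen-space refraction are also computed.
# The attributes feeding destructible damage use constant magenta base color, leaving emissive and refraction inputs unconnected.
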Material /Hermes/Props/BigSplat/BigSplat_LootSac_A/Materials/Backup_BigSplat_LootSacBlob_Translucent
graph {
 properties {bAutomaticallySetUsageInEditor=false, bUseMaterialAttributes=true, blendMode=BLEND_Translucent, usedShadingModels=Default Lit}
 nodes [
  n1 = Time()
  n2 = ScalarParameter() {defaultValue=-0.005, parameterName=Liquid1_ScrollX}
  n3 = ScalarParameter() {defaultValue=0.01, parameterName=Liquid1_ScrollY}
  n4 = Add(A=MaterialFunctionCall!(Position=ObjectPositionWS(), UVRadialScale) {materialFunction=/Game/Weapons/WeaponSkins/MaterialFunction/ObjectStabilizedScreenSpaceUVs}.StabilizedUVs) {constB=-0.5}
  n5 = NamedRerouteDeclaration(None=ComponentMask(None=Multiply(A=@n4) {constB=-1}) {g=true, r=true}) {name=BumpOffsetUV, nodeColor=<rgba 0.125 0 1 1>}
  n6 = Add(A=Constant() {r=-0.3}, B=Multiply(A=NamedRerouteUsage(), B=ScalarParameter() {defaultValue=0.1, parameterName=Liquid_Distortion_1}))
  n7 = Add(A=Add(A=NamedRerouteUsage(), B=Multiply(A=@n5, B=@n6)), B=AppendVector(A=Multiply(A=@n1, B=@n2), B=Multiply(A=@n1, B=@n3)))
  n8 = Multiply(A=@n7, B=ScalarParameter() {defaultValue=2, parameterName=Liquid1_TileMask})
  Add(A=@n8, B=NamedRerouteUsage())
  n9 = Frac(None=Multiply(A=@n7, B=ScalarParameter() {defaultValue=2, desc=3, parameterName=Liquid1_Tile}))
  Add(A=@n9, B=NamedRerouteUsage())
  n10 = NamedRerouteUsage()
  n11 = Add(
   A=ScalarParameter() {defaultValue=-0.7, parameterName=Liquid_Distorion_BumpO}
   B=Multiply(A=NamedRerouteUsage(), B=ScalarParameter() {defaultValue=0.25, parameterName=Liquid_Distortion_2})
  )
  n12 = Time()
  n13 = ScalarParameter() {defaultValue=-0.003, parameterName=Liquid2_ScrollX}
  n14 = ScalarParameter() {defaultValue=0.01, parameterName=Liquid2_ScrollY}
  n15 = Add(A=Add(A=NamedRerouteUsage(), B=Multiply(A=@n10, B=@n11)), B=AppendVector(A=Multiply(A=@n12, B=@n13), B=Multiply(A=@n12, B=@n14)))
  n16 = Multiply(A=@n15, B=ScalarParameter() {defaultValue=3, desc=6, parameterName=Liquid2_Tile})
  Add(A=@n16, B=NamedRerouteUsage())
  n17 = Add(A=Multiply(A=CameraVectorWS()) {constB=-1}, B=Multiply(A=VertexNormalWS(), B=Constant() {r=0.5}))
  Add(A=@n17, B=NamedRerouteUsage())
  n18 = Multiply(A=@n15, B=ScalarParameter() {defaultValue=1, parameterName=Liquid2_TileMask})
  Add(A=@n18, B=NamedRerouteUsage())
  n19 = NamedRerouteUsage()
  n20 = Max(
   A=Max(A=NamedRerouteUsage(), B=NamedRerouteUsage())
   B=LinearInterpolate(
    A=ScalarParameter() {defaultValue=0.3, group=Opacity, parameterName=Opacity_Fresnel_Center}
    B=ScalarParameter() {defaultValue=1, group=Opacity, parameterName=Opacity_Fresnel_Borders}
    Alpha=Saturate(None=@n19)
   )
  )
  LinearInterpolate(A=ScalarParameter() {defaultValue=0.3, parameterName=Opacity_Glass}, B=@n20, Alpha=NamedRerouteUsage())
  n21 = SetMaterialAttributes!(
   MaterialAttributes
   Base Color=Constant3Vector() {constant=<rgba 1 0 0.9326 1>}
   Emissive Color
   Specular=ScalarParameter() {defaultValue=1, parameterName=Specular}
   Roughness=ScalarParameter() {parameterName=Roughness}
   Metallic=ScalarParameter() {parameterName=Metallic}
   Opacity=NamedRerouteUsage()
   Normal
   Refraction
  ) {attributeSetTypes [
    #1
    #2
    #3
    #4
    #5
    #6
    #7
    #8
   ]}
  MaterialFunctionCall(Material_In=@n21) {materialFunction=&MF_FN_DestructibleDamage}
  n22 = MaterialFunctionCall() {materialFunction=&CameraDirectionVector}
  n23 = Clamp(None=LinearInterpolate(A=Constant() {r=0.3}, B=Constant() {desc=-0.8, r=-0.3}, Alpha=ComponentMask(None=@n22.Result) {b=true})) {minDefault=0.06}
  n24 = Constant() {r=0.05}
  n25 = NamedRerouteUsage()
  n26 = Multiply(A=NamedRerouteUsage()) {constB=2}
  n27 = Abs(None=ComponentMask(None=@n26) {r=true})
  n28 = Multiply(A=@n27, B=@n27)
  n29 = Multiply(
   A=SmoothStep(Min=Constant() {r=-0.3}, Max=Constant() {r=-3}, Value=ComponentMask(None=@n26) {g=true})
   B=SmoothStep(Min=Constant() {r=1}, Max=Constant(), Value=@n28)
  )
  Max(
   A=NamedRerouteDeclaration(None=SmoothStep(Min=@n23, Max=@n24, Value=@n25) {constMax=0.05, constMin=0.1}) {name=Hacky_bottomMask, nodeColor=<rgba 1 0.2109 0 1>}
   B=@n29
  )
  n30 = VectorParameter() {parameterName=Glass_Color Day, defaultValue=<rgba 0.4913 0.4642 0.2636 0.7>}
  n31 = MaterialFunctionCall(
   Color_Morning=@n30.RGB
   Color_Day=@n30.RGB
   Color_Evening=@n30.RGB
   Color_Night=VectorParameter() {parameterName=Glass_Color Night, defaultValue=<rgba 0.489 0.9768 1 0.7>}.RGB
  ) {materialFunction=&Color_Time_of_Day}
  n32 = NamedRerouteUsage()
  n33 = Constant() {r=0.5}
  n34 = MaterialFunctionCall(
   Color_Morning=VectorParameter() {parameterName="Liquid_Color_Morning3(sun)", defaultValue=<rgba 10 3.1252 5.1116 1>}.RGB
   Color_Day=VectorParameter() {parameterName="Liquid_Color_Day3(sun)", defaultValue=<rgba 2.2057 3 1.5833 1>}.RGB
   Color_Evening=VectorParameter() {parameterName="Liquid_Color_Evening(sun)", defaultValue=<rgba 6 1.5635 1.3047 1>}.RGB
   Color_Night=VectorParameter() {parameterName="Liquid_Color_Night(sun)", defaultValue=<rgba 4.5 5.598 6 1>}.RGB
  ) {materialFunction=&Color_Time_of_Day}
  n35 = Multiply(A=@n34.Result) {constB=0.4}
  n36 = MaterialFunctionCall(
   Color_Morning=LinearInterpolate(
    A=VectorParameter() {parameterName=Liquid_Color_Morning1, defaultValue=<rgba 0.0411 0.092 0 1>}.RGB
    B=VectorParameter() {parameterName=Liquid_Color_Morning2, defaultValue=<rgba 1 0.8304 0 1>}.RGB
    Alpha=NamedRerouteUsage()
   )
   Color_Day=LinearInterpolate(
    A=VectorParameter() {parameterName=Liquid_Color_Day1, defaultValue=<rgba 0.3089 0.4149 0.0425 1>}.RGB
    B=VectorParameter() {parameterName=Liquid_Color_Day2, defaultValue=<rgba 0.9402 1 0.1878 1>}.RGB
    Alpha=NamedRerouteUsage()
   )
   Color_Evening=LinearInterpolate(
    A=VectorParameter() {parameterName=Liquid_Color_Evening1, defaultValue=<rgba 0.0253 0.0887 0.0266 1>}.RGB
    B=VectorParameter() {parameterName=Liquid_Color_Evening2, defaultValue=<rgba 0.8439 1 0.386 1>}.RGB
    Alpha=NamedRerouteUsage()
   )
   Color_Night=LinearInterpolate(
    A=VectorParameter() {parameterName=Liquid_Color_Night1, defaultValue=<rgba 0.004 0.0764 0.039 1>}.RGB
    B=VectorParameter() {parameterName=Liquid_Color_Night2, defaultValue=<rgba 0.1816 0.934 0.8786 1>}.RGB
    Alpha=NamedRerouteUsage()
   )
  ) {materialFunction=&Color_Time_of_Day}
  n37 = LinearInterpolate(A=@n36.Result, B=Multiply(A=@n36.Result, B=Constant() {r=0.5}), Alpha=Add(A=NamedRerouteUsage(), B=Constant()))
  n38 = LinearInterpolate(A=@n37, B=Multiply(A=@n37, B=ShadingPathSwitch(Default=@n34.Result, Mobile=@n35)), Alpha=Add(A=NamedRerouteUsage(), B=Constant()))
  n39 = LinearInterpolate(
   A=@n38
   B=Multiply(A=@n38, B=VectorParameter() {parameterName=Liquid_SurfaceColor, defaultValue=<rgba 1.5 1.1896 0.174 1>}.RGB)
   Alpha=NamedRerouteUsage()
  )
  n40 = Min(A=NamedRerouteUsage(), B=NamedRerouteUsage())
  n41 = NamedRerouteUsage()
  n42 = LinearInterpolate(
   A=LinearInterpolate(A=@n31.Result, B=Multiply(A=@n31.Result, B=@n32), Alpha=ScalarParameter() {defaultValue=0.9, parameterName=ColorBlend})
   B=LinearInterpolate(A=@n39, B=Multiply(A=@n39, B=@n33), Alpha=NamedRerouteUsage())
   Alpha=Saturate(None=Subtract(A=@n40, B=@n41))
  )
  Multiply(A=@n42, B=@n20)
  n43 = SceneDepth()
  n44 = PixelDepth()
  n45 = MaterialFunctionCall(Color=ObjectBounds()) {materialFunction=&SplitComponents}
  NamedRerouteDeclaration(None=Divide(A=Subtract(A=@n43.R, B=@n44.R), B=Max(A=@n45.R, B=@n45.R))) {name=M_Depht, nodeColor=<rgba 0 1 0.2734 1>}
  n46 = MaterialFunctionCall() {materialFunction=&CameraDirectionVector}
  n47 = Add(A=@n28, B=LinearInterpolate(A=Constant() {r=2}, B=Constant() {r=3.5}, Alpha=ComponentMask(None=@n46.Result) {b=true}))
  n48 = Abs(None=ComponentMask(None=NamedRerouteUsage()) {r=true})
  n49 = Constant() {desc=0.1, r=0.1}
  n50 = Max(
   A=Multiply(A=@n29, B=Saturate(None=@n47))
   B=MaterialFunctionCall(In=Add(A=@n48, B=@n49), Contrast=Constant() {desc=0.5, r=0.5}) {materialFunction=&CheapContrast}.Result
  )
  n51 = Add(A=NamedRerouteUsage(), B=Constant() {desc=-0.45, r=-0.55})
  NamedRerouteDeclaration(None=MaterialFunctionCall(In=Add(A=@n50, B=@n51), Contrast=Constant() {r=0.3}) {materialFunction=&CheapContrast}.Result) {name=BG_Mask, nodeColor=<rgba 0 0.2734 1 1>}
  n52 = Subtract(A=WorldPosition() {worldPositionShaderOffset=WPT_ExcludeAllShaderOffsets}.XYZ, B=ObjectPositionWS())
  n53 = ObjectLocalBounds()
  n54 = Divide(A=Add(A=@n52, B=@n53.Half Extents) {constB=0.5}, B=@n53.Extents)
  n55 = AppendVector(
   A=MaterialFunctionCall(Vector or UVs=ComponentMask(None=@n52) {g=true, r=true}) {materialFunction=&VectorToRadialValue}.Vector Converted to Angle
   B=NamedRerouteDeclaration(None=ComponentMask(None=@n54) {b=true}) {name=Local H, nodeColor=<rgba 0.5469 1 0 1>}
  )
  NamedRerouteDeclaration(None=Add(A=Multiply(A=@n55) {constB=-1})) {name=UV_Cylindrical, nodeColor=<rgba 1 0.8438 0 1>}
  n56 = Max(
   A=TextureSample!(UVs=@n9, Tex, "DDX(UVs)"=NamedRerouteUsage(), "DDY(UVs)"=NamedRerouteUsage(), Apply View MipBias) {mipValueMode=TMVM_Derivative, texture=/Hermes/Props/CozyTea/CozyTea_PickleSign_A/Textures/T_CozyTea_PickleSign_M}.RGB
   B=TextureSample(UVs=@n16, "DDX(UVs)"=NamedRerouteUsage(), "DDY(UVs)"=NamedRerouteUsage()) {mipValueMode=TMVM_Derivative, texture=&T_CozyTea_PickleSign_M}.RGB
  )
  n57 = ComponentMask(None=@n56) {g=true}
  NamedRerouteDeclaration(None=@n57) {name=M_SurfaceDistortion, nodeColor=<rgba 0.3594 1 0 1>}
  n58 = TextureCoordinate()
  NamedRerouteDeclaration(None=DDX(Value=@n58)) {name=DDX, nodeColor=<rgba 0 0.2734 1 1>}
  NamedRerouteDeclaration(None=DDY(Value=@n58)) {name=DDY, nodeColor=<rgba 0 0.2734 1 1>}
  n59 = Multiply(
   A=SkyLightEnvMapSample(Direction=@n17, Roughness=ScalarParameter() {group=CubeMap, parameterName=SkylightEnvMapRoughness})
   B=LinearInterpolate(
    A=ScalarParameter() {defaultValue=1, desc=0.03, parameterName=BG_CubemapMult_SidesAir}
    B=VectorParameter() {parameterName=Cubemap Color Lerp, defaultValue=<rgba 0.1845 0.23 0.1007 1>}.RGB
    Alpha=NamedRerouteUsage()
   )
  )
  n60 = NamedRerouteUsage()
  NamedRerouteDeclaration(None=Multiply(A=LinearInterpolate(B=@n59, Alpha=@n60), B=NamedRerouteUsage())) {name=BaseColor, nodeColor=<rgba 1 0.3516 0 1>}
  n61 = Constant()
  n62 = ScalarParameter() {defaultValue=650, parameterName=Liquid_Surface_Height}
  n63 = ScalarParameter() {defaultValue=4, parameterName=Luquid_Surface Waves}
  n64 = MaterialFunctionCall(R=@n61, G=@n61, B=Subtract(A=@n62, B=@n63)) {materialFunction=&AppendMany}
  n65 = ScalarParameter() {defaultValue=730, parameterName=Liquid_Surface_Size}
  n66 = Distance(
   A=MaterialFunctionCall(
    Plane X Axis=Constant3Vector() {constant=<rgba 0 1 0 0>}
    Plane Center=TransformPosition(None=@n64.RGB) {transformType=TRANSFORMPOSSOURCE_World}
    UV Scale=AppendVector(A=@n65, B=@n65)
   ) {materialFunction=&VirtualPlaneCoordinates}.UVs
   B=Constant() {r=0.5}
  )
  n67 = OneMinus(None=Step(Y=@n66) {constX=0})
  n68 = Constant() {r=0.5}
  n69 = Add(A=LocalPosition().Z, B=Multiply(A=NamedRerouteUsage(), B=@n63))
  n70 = Step(Y=@n69, X=@n62) {constX=0.9}
  n71 = Subtract(A=Multiply(A=Saturate(None=@n67), B=Step(Y=@n66, X=@n68) {constX=0.5}), B=@n70)
  n72 = CameraPositionWS()
  n73 = TransformPosition(None=Add(A=@n62, B=Constant3Vector())) {transformType=TRANSFORMPOSSOURCE_World}
  n74 = Step(X=ComponentMask(None=Subtract(A=@n72, B=@n73)) {b=true}) {constX=0}
  n75 = NamedRerouteDeclaration(None=Multiply(A=Saturate(None=@n71), B=NamedRerouteDeclaration(None=@n74) {name=CameraToLiquid, nodeColor=<rgba 0.6875 1 0 1>})) {name=Liquid Surface Mask, nodeColor=<rgba 1 0.5156 0 1>}
  n76 = NamedRerouteDeclaration(None=@n70) {name=Liquid Sides Mask, nodeColor=<rgba 1 0.3984 0 1>}
  n77 = SmoothStep(Min=Constant() {r=0.8}, Max=Constant() {r=0.5}, Value=NamedRerouteUsage())
  NamedRerouteDeclaration(None=Min(A=Max(A=@n75, B=@n76), B=NamedRerouteDeclaration(None=@n77) {name=Glass Thickness, nodeColor=<rgba 1 0 0.8906 1>})) {name=Liquid Silhouete Mask, nodeColor=<rgba 0 0.6719 1 1>}
  NamedRerouteDeclaration(None=Multiply(A=@n42, B=EyeAdaptationInverse().EyeAdaptationInverse)) {name=Emissive, nodeColor=<rgba 1 0 0.5391 1>}
  n78 = Subtract(A=@n62, B=ScalarParameter() {defaultValue=5, parameterName=Liquid_Thickness})
  NamedRerouteDeclaration(None=Multiply(A=@n70, B=Step(Y=@n78, X=@n69) {constX=0.9})) {name=Liquid_Thickness, nodeColor=<rgba 0 1 0.6719 1>}
  NamedRerouteDeclaration(None=@n20) {name=Opacity, nodeColor=<rgba 0.125 0 1 1>}
  n79 = Multiply(A=DotProduct(A=SkyAtmosphereLightDirection(), B=CameraVectorWS())) {constB=-1}
  n80 = ScalarParameter() {defaultValue=-0.7, parameterName=SunAuraSize}
  n81 = Saturate(None=Multiply(A=Add(A=@n79, B=@n80), B=Constant() {r=3}))
  NamedRerouteDeclaration(None=Multiply(A=@n81, B=@n81)) {name=SunBlob, nodeColor=<rgba 1 0 0.8203 1>}
  n82 = Add(A=NamedRerouteUsage(), B=Multiply(A=NamedRerouteUsage(), B=ScalarParameter() {defaultValue=-0.7, parameterName=Liquid_Distorion_BumpO}))
  n83 = AppendVector(A=Multiply(A=Time(), B=ScalarParameter() {defaultValue=-0.01, parameterName=Liquid_Distortion_Scroll}) {constB=0}, B=Constant())
  n84 = TextureSample(
   UVs=Multiply(A=Add(A=@n82, B=@n83), B=ScalarParameter() {defaultValue=1, parameterName=Liquid_Distortion_Tile})
   "DDX(UVs)"=NamedRerouteUsage()
   "DDY(UVs)"=NamedRerouteUsage()
  ) {mipValueMode=TMVM_Derivative, samplerType=SAMPLERTYPE_Grayscale, texture=/Game/Global/Textures/Default/Noise/LowResBlurredNoiseNormalized}
  NamedRerouteDeclaration(None=@n84.R) {name=Distortion Texture, nodeColor=<rgba 1 0.4219 0 1>}
  n85 = ScalarParameter() {defaultValue=1, group=Roughness, parameterName=Rough_RepeatU}
  n86 = ScalarParameter() {defaultValue=1, group=Roughness, parameterName=Rough_RepeatV}
  n87 = TextureSampleParameter2D(UVs=Multiply(A=TextureCoordinate(), B=AppendVector(A=@n85, B=@n86))) {
   group=Grime
   parameterName=TextureMask
   samplerType=SAMPLERTYPE_Masks
   texture=/Game/Global/Materials/EnvironmentGlass/T_GlassTestMask_Indestruct
   channelNames {a="", b=Mask, g=Roughness, r=Grime}
  }
  NamedRerouteDeclaration(None=@n87.Roughness) {name=RoughMask, nodeColor=<rgba 0.602 0.7917 0.7205 1>}
  n88 = ScalarParameter() {
   defaultValue=1.5
   desc="Tighten/loosen the fresnel  blend.  Fresnel allows for different values of opacity and roughness based on the angle of the surface facing the camera. (Glass is shinier when looking at it from the edge)"
   parameterName=Fresnel Exponent
  }
  NamedRerouteDeclaration(None=Fresnel(ExponentIn=@n88) {baseReflectFraction=0}) {name=FresnelMask, nodeColor=<rgba 1 0.8814 0 1>}
  n89 = ComponentMask(None=@n56) {r=true}
  n90 = Constant() {r=1}
  n91 = Add(A=@n57, B=Constant() {r=1})
  n92 = Constant() {r=0.2}
  n93 = NamedRerouteUsage()
  n94 = NamedRerouteUsage()
  n95 = NamedRerouteUsage()
  n96 = NamedRerouteUsage()
  n97 = Add(
   A=Multiply(
    A=TextureSample(UVs=@n18, "DDX(UVs)"=@n93, "DDY(UVs)"=@n94) {mipValueMode=TMVM_Derivative, texture=&T_CozyTea_PickleSign_M}.B
    B=Constant() {r=1}
   ) {constB=0.5}
   B=Multiply(
    A=TextureSample(UVs=@n8, "DDX(UVs)"=@n95, "DDY(UVs)"=@n96) {mipValueMode=TMVM_Derivative, texture=&T_CozyTea_PickleSign_M}.B
    B=Constant() {r=0.8}
   ) {constB=0.2}
  )
  n98 = ScalarParameter() {defaultValue=-0.3, desc=-.3, parameterName=Liquid_Viscosity_Bias}
  n99 = LinearInterpolate(
   A=Max(A=Multiply(A=@n89, B=@n90), B=Multiply(A=@n91, B=@n92))
   Alpha=MaterialFunctionCall(In=Add(A=@n97, B=@n98), Contrast=ScalarParameter() {defaultValue=-0.35, desc=-.2, parameterName=Liquid_Visocity_Contrast}) {materialFunction=&CheapContrast}.Result
  ) {constB=0}
  n100 = NamedRerouteUsage()
  n101 = LinearInterpolate(A=Constant() {r=-0.2}, B=Constant() {r=2}, Alpha=NamedRerouteUsage())
  NamedRerouteDeclaration(None=Multiply(A=Multiply(A=@n99, B=@n100), B=Saturate(None=@n101))) {name=Liquid Interior Pattern, nodeColor=<rgba 0.875 1 0 1>}
  n102 = SkyLightEnvMapSample(Direction=ReflectionVectorWS(), Roughness=ScalarParameter() {group=CubeMap, parameterName=SkylightEnvMapRoughness})
  n103 = ScalarParameter() {
   defaultValue=0.5
   desc="A multiplier on the cubemap to blend it in and out.  Higher values show it more."
   group=CubeMap
   parameterName=CubeMap_Mult
  }
  n104 = ScalarParameter() {
   defaultValue=0.8
   desc=When the angle of the surface faces more away from the camera this will be the value used for Cubemap mult.
   group=CubeMap
   parameterName=CubeMap_Fresnel
  }
  n105 = NamedRerouteUsage()
  n106 = ScalarParameter() {defaultValue=0.5, group=Roughness, parameterName=RoughMask_Value_MobileForward}
  n107 = ShadingPathSwitch(Default=ScalarParameter() {defaultValue=0.5, group=Roughness, parameterName=RoughMask_Value}, Forward=@n106, Mobile=@n106)
  n108 = LinearInterpolate(A=LinearInterpolate(A=@n103, B=@n104, Alpha=NamedRerouteUsage()), Alpha=Saturate(None=Multiply(A=@n105, B=@n107))) {constB=0}
  NamedRerouteDeclaration(None=Multiply(A=Multiply(A=@n108, B=@n102), B=VectorParameter() {group=CubeMap, parameterName=CubemapColorScale, defaultValue=<rgba 1 1 1 0>}.RGB)) {name=CubeMapColor_Reflex, nodeColor=<rgba 0.2422 0 1 1>}
  NamedRerouteUsage()
  NamedRerouteUsage()
  NamedRerouteUsage()
  NamedRerouteUsage()
  NamedRerouteUsage()
  n109 = ScreenPosition()
  n110 = Constant() {r=2}
  n111 = Constant() {r=0.2}
  n112 = TextureSample(UVs=Multiply(A=NamedRerouteUsage(), B=AppendVector(A=@n110, B=@n111)), "DDX(UVs)"=NamedRerouteUsage(), "DDY(UVs)"=NamedRerouteUsage()) {mipValueMode=TMVM_Derivative, texture=&T_CozyTea_PickleSign_M}
  n113 = Constant() {desc=.02, r=0.03}
  n114 = Constant() {desc=.005, r=0.005}
  n115 = Distance(
   A=CameraPositionWS()
   B=MaterialFunctionCall() {materialFunction=&ObjectPivotPoint}.Object Pivot Location
  )
  n116 = Constant() {r=300}
  n117 = Saturate(None=Divide(A=Subtract(A=@n115, B=@n116) {constB=3000}, B=Constant() {r=3000}) {constB=1500})
  n118 = LinearInterpolate(
   A=Multiply(A=Add(A=@n112.G) {constB=-0.5}, B=LinearInterpolate(A=@n113, B=@n114, Alpha=@n117))
   B=Constant()
   Alpha=SmoothStep(Min=Constant() {r=0.1}, Max=Constant() {r=0.5}, Value=NamedRerouteUsage())
  )
  n119 = SmoothStep(Value=LocalPosition().Z) {constMax=600, constMin=680, constValue=600}
  n120 = Subtract(A=SceneDepth().R, B=Distance(A=WorldPosition().XYZ, B=CameraPositionWS()))
  n121 = NamedRerouteDeclaration(None=Multiply(A=Multiply(A=@n118, B=@n119), B=Saturate(None=@n120))) {name=Refraction, nodeColor=<rgba 1 0.2578 0 1>}
  SceneColor!(UVs=NamedRerouteDeclaration(None=Add(A=@n109.ViewportUV, B=@n121)) {name=DistortedScreenUV, nodeColor=<rgba 0 1 0.4375 1>})
 ]
}
