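# Material function that outputs an animated World Position Offset wobble, named for the blob on the BigSplat loot sack prop.
# A blurred noise texture is sampled with local-position XZ and YZ projections blended by absolute vertex normal, and scrolled over time at different rates per axis.
# The noise scales a displacement along the vertex normal, with a static InvertNormal switch to flip its direction.
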MaterialFunction /Hermes/Props/BigSplat/BigSplat_LootSac_A/Materials/MF_BigSplat_LootSackBLob_WPO
graph {nodes [
  Constant3Vector() {constant=<rgba 1 0.6581 0.1298 1>}
  n1 = LocalPosition()
  n2 = LinearInterpolate(A=ComponentMask(None=@n1.XYZ) {b=true, r=true}, B=ComponentMask(None=@n1.XYZ) {b=true, g=true}, Alpha=Abs(None=VertexNormalWS()))
  n3 = Constant() {r=180}
  n4 = Time()
  n5 = Divide(A=@n4, B=Constant() {r=-5})
  n6 = Divide(A=@n4, B=Constant() {r=-20})
  n7 = TextureSample(UVs=Add(A=Divide(A=@n2, B=@n3) {constB=200}, B=AppendVector(A=@n5, B=@n6))) {samplerType=SAMPLERTYPE_Grayscale, texture=/Game/Global/Textures/Default/Noise/LowResBlurredNoiseNormalized}
  n8 = VertexNormalWS()
  n9 = StaticSwitch(
   True=Multiply(A=@n8) {constB=-1}
   False=@n8
   Value=FunctionInput(Preview=StaticBool()) {bUsePreviewValueAsDefault=true, inputName=InvertNormal, inputType=FunctionInput_StaticBool}
  )
  n10 = FunctionOutput(None=Multiply(A=Multiply(A=@n7.R, B=@n9), B=Constant() {r=2.5})) {outputName=WPO}
  SetMaterialAttributes!(MaterialAttributes, WorldPositionOffset=@n10, BaseColor=@n10) {attributeSetTypes [#1, #2]}
 ]}
