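# Material function that layers a glowing liquid-sack look onto the BigSplat loot sac prop's small sacks, blended by vertex-color red behind a static switch.
# Two scrolling, noise-distorted texture layers, using cylindrical and object-stabilized screen-space UVs for parallax, form the liquid interior.
# Color ramps, Fresnel, membrane and sun-direction masks shape an exposure-compensated emissive.
# The sack layer also sets specular, roughness and metallic, scales normals and adds blob world position offset.
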MaterialFunction /Hermes/Props/BigSplat/BigSplat_LootSac_A/Materials/MF_BigSplat_SmallSacks
graph {nodes [
  n1 = GetMaterialAttributes(None=FunctionInput() {inputName=MA, inputType=FunctionInput_MaterialAttributes}) {attributeGetTypes [#1, #2, #3]}
  n2 = ScalarParameter() {defaultValue=0.5, group=TidalCrane_Sacs, parameterName=Sacks_Normal}
  n3 = Constant() {r=1}
  n4 = SetMaterialAttributes!(
   MaterialAttributes=@n1.MaterialAttributes
   BaseColor=NamedRerouteUsage()
   EmissiveColor=NamedRerouteUsage()
   Specular=ScalarParameter() {defaultValue=1, group=TidalCrane_Sacs, parameterName=Sacks_Specular}
   Roughness=ScalarParameter() {defaultValue=0.2, group=TidalCrane_Sacs, parameterName=Sacks_Roughness}
   Metallic=ScalarParameter() {defaultValue=1, group=TidalCrane_Sacs, parameterName=Sacks_Metallic}
   Opacity
   Normal=Multiply(
    A=@n1.Normal
    B=MaterialFunctionCall(R=@n2, G=@n2, B=@n3) {materialFunction=&AppendMany}.RGB
   )
   Refraction
   WorldPositionOffset=Add(
    A=@n1.WorldPositionOffset
    B=MaterialFunctionCall!(InvertNormal) {materialFunction=&/MF_BigSplat_LootSackBLob_WPO}.WPO
   )
  ) {attributeSetTypes [
    #1
    #4
    #5
    #6
    #7
    #8
    #2
    #9
    #3
   ]}
  n5 = VertexColor()
  FunctionOutput(None=StaticSwitchParameter(True=BlendMaterialAttributes(A=@n1.MaterialAttributes, B=@n4, Alpha=@n5.R), False=@n1.MaterialAttributes) {defaultValue=true, group=TidalCrane_Sacks, parameterName=Use_SacksMaterial, sortPriority=0}) {outputName=BlendMaterialAttributes}
  n6 = Subtract(A=PixelDepth().R, B=ScalarParameter() {defaultValue=380, group=TidalCrane_Sacs}) {constB=800}
  n7 = ScalarParameter() {group=TidalCrane_Sacs, parameterName=Param_3}
  LinearInterpolate(A=@n5.R, Alpha=Saturate(None=Divide(A=@n6, B=@n7) {constB=0})) {constB=0}
  n8 = NamedRerouteUsage()
  n9 = Multiply(A=NamedRerouteUsage(), B=ScalarParameter() {defaultValue=0.2, parameterName=Param_23})
  n10 = Add(A=Add(A=@n8, B=Subtract(A=@n8, B=@n9)), B=ScalarParameter() {defaultValue=0.52, parameterName=Param_25}) {constB=0.2}
  n11 = ScalarParameter() {defaultValue=2, desc=3, parameterName=Param_24}
  NamedRerouteDeclaration(None=Saturate(None=MaterialFunctionCall(In=@n10, Contrast=@n11) {materialFunction=&CheapContrast}.Result)) {name=M_Depht, nodeColor=<rgba 0 1 0.2734 1>}
  NamedRerouteDeclaration(None=LinearInterpolate(
    A=VectorParameter() {parameterName=Cubemap Color Lerp, defaultValue=<rgba 0.255 0.1068 0.0341 1>}.RGB
    B=ScalarParameter() {defaultValue=0.4, desc=0.03, parameterName=BG_CubemapMult_SidesAir}
    Alpha=NamedRerouteUsage()
   )) {name=BaseColor, nodeColor=<rgba 1 0.3516 0 1>}
  n12 = ScalarParameter() {
   defaultValue=1.5
   desc="Tighten/loosen the fresnel  blend.  Fresnel allows for different values of opacity and roughness based on the angle of the surface facing the camera. (Glass is shinier when looking at it from the edge)"
   parameterName=Fresnel Exponent
  }
  NamedRerouteDeclaration(None=Fresnel(ExponentIn=@n12) {baseReflectFraction=0}) {name=FresnelMask, nodeColor=<rgba 1 0.8814 0 1>}
  n13 = Subtract(A=WorldPosition() {worldPositionShaderOffset=WPT_ExcludeAllShaderOffsets}.XYZ, B=ObjectPositionWS())
  n14 = ObjectLocalBounds()
  n15 = Divide(A=Add(A=@n13, B=@n14.Half Extents) {constB=0.5}, B=@n14.Extents)
  n16 = AppendVector(
   A=MaterialFunctionCall(Vector or UVs=ComponentMask(None=@n13) {g=true, r=true}) {materialFunction=&VectorToRadialValue}.Vector Converted to Angle
   B=NamedRerouteDeclaration(None=ComponentMask(None=@n15) {b=true}) {name=Local H, nodeColor=<rgba 0.5469 1 0 1>}
  )
  NamedRerouteDeclaration(None=Add(A=Multiply(A=@n16) {constB=-1})) {name=UV_Cylindrical, nodeColor=<rgba 1 0.8438 0 1>}
  n17 = TextureCoordinate()
  NamedRerouteDeclaration(None=DDX(Value=@n17)) {name=DDX, nodeColor=<rgba 0 0.2734 1 1>}
  NamedRerouteDeclaration(None=DDY(Value=@n17)) {name=DDY, nodeColor=<rgba 0 0.2734 1 1>}
  n18 = Multiply(A=DotProduct(A=SkyAtmosphereLightDirection(), B=CameraVectorWS())) {constB=-1}
  n19 = ScalarParameter() {defaultValue=-0.7, parameterName=SunAuraSize}
  n20 = Saturate(None=Multiply(A=Add(A=@n18, B=@n19), B=Constant() {r=3}))
  NamedRerouteDeclaration(None=Multiply(A=@n20, B=@n20)) {name=SunBlob, nodeColor=<rgba 1 0 0.8203 1>}
  n21 = Time()
  n22 = ScalarParameter() {defaultValue=-0.005, parameterName=Liquid1_ScrollX}
  n23 = ScalarParameter() {defaultValue=0.01, parameterName=Liquid1_ScrollY}
  n24 = Add(A=MaterialFunctionCall!(Position=ObjectPositionWS(), UVRadialScale) {materialFunction=/Game/Weapons/WeaponSkins/MaterialFunction/ObjectStabilizedScreenSpaceUVs}.StabilizedUVs) {constB=-0.5}
  n25 = NamedRerouteDeclaration(None=ComponentMask(None=Multiply(A=@n24) {constB=-1}) {g=true, r=true}) {name=BumpOffsetUV, nodeColor=<rgba 0.125 0 1 1>}
  n26 = Add(A=Constant() {r=-0.3}, B=Multiply(A=NamedRerouteUsage(), B=ScalarParameter() {defaultValue=0.1, parameterName=Liquid_Distortion_1}))
  n27 = Add(A=Add(A=NamedRerouteUsage(), B=Multiply(A=@n25, B=@n26)), B=AppendVector(A=Multiply(A=@n21, B=@n22), B=Multiply(A=@n21, B=@n23)))
  n28 = Multiply(A=@n27, B=ScalarParameter() {defaultValue=2, desc=3, parameterName=Liquid1_Tile})
  n29 = NamedRerouteUsage()
  n30 = Add(
   A=ScalarParameter() {defaultValue=-0.7, parameterName=Liquid_Distorion_BumpO}
   B=Multiply(A=NamedRerouteUsage(), B=ScalarParameter() {defaultValue=0.25, parameterName=Liquid_Distortion_2})
  )
  n31 = Time()
  n32 = ScalarParameter() {defaultValue=-0.003, parameterName=Liquid2_ScrollX}
  n33 = ScalarParameter() {defaultValue=0.01, parameterName=Liquid2_ScrollY}
  n34 = Add(A=Add(A=NamedRerouteUsage(), B=Multiply(A=@n29, B=@n30)), B=AppendVector(A=Multiply(A=@n31, B=@n32), B=Multiply(A=@n31, B=@n33)))
  n35 = ScalarParameter() {defaultValue=4, desc=6, parameterName=Liquid2_Tile}
  n36 = Max(
   A=TextureSample!(UVs=Frac(None=@n28), Tex, "DDX(UVs)"=NamedRerouteUsage(), "DDY(UVs)"=NamedRerouteUsage(), Apply View MipBias) {mipValueMode=TMVM_Derivative, texture=/Hermes/Props/CozyTea/CozyTea_PickleSign_A/Textures/T_CozyTea_PickleSign_M}.RGB
   B=TextureSample(UVs=Multiply(A=@n34, B=@n35), "DDX(UVs)"=NamedRerouteUsage(), "DDY(UVs)"=NamedRerouteUsage()) {mipValueMode=TMVM_Derivative, texture=&T_CozyTea_PickleSign_M}.RGB
  )
  n37 = ComponentMask(None=@n36) {r=true}
  n38 = Constant() {r=1}
  n39 = Add(A=ComponentMask(None=@n36) {g=true}, B=Constant() {r=1})
  n40 = Constant() {r=0.2}
  n41 = Multiply(A=@n34, B=ScalarParameter() {defaultValue=1, parameterName=Liquid2_TileMask})
  n42 = NamedRerouteUsage()
  n43 = NamedRerouteUsage()
  n44 = TextureSample(
   UVs=Multiply(A=@n27, B=ScalarParameter() {defaultValue=2, parameterName=Liquid1_TileMask})
   "DDX(UVs)"=NamedRerouteUsage()
   "DDY(UVs)"=NamedRerouteUsage()
  ) {mipValueMode=TMVM_Derivative, texture=&T_CozyTea_PickleSign_M}
  n45 = Add(
   A=Multiply(
    A=TextureSample(UVs=@n41, "DDX(UVs)"=@n42, "DDY(UVs)"=@n43) {mipValueMode=TMVM_Derivative, texture=&T_CozyTea_PickleSign_M}.B
    B=Constant() {r=1}
   ) {constB=0.5}
   B=Multiply(A=NamedRerouteDeclaration(None=@n44.B) {name=Noise_2, nodeColor=<rgba 0.1953 1 0 1>}, B=Constant() {r=0.8}) {constB=0.2}
  )
  n46 = ScalarParameter() {defaultValue=-0.3, desc=-.3, parameterName=Liquid_Viscosity_Bias}
  n47 = LinearInterpolate(
   A=Max(A=Multiply(A=@n37, B=@n38), B=Multiply(A=@n39, B=@n40))
   Alpha=MaterialFunctionCall(In=Add(A=@n45, B=@n46), Contrast=ScalarParameter() {defaultValue=-0.35, desc=-.2, parameterName=Liquid_Visocity_Contrast}) {materialFunction=&CheapContrast}.Result
  ) {constB=0}
  n48 = Saturate(None=LinearInterpolate(A=Constant() {r=-0.2}, B=Constant() {r=2}, Alpha=NamedRerouteUsage()))
  n49 = SmoothStep(Min=Constant() {r=0.02}, Max=Constant() {r=0.08}, Value=ComponentMask(None=NamedRerouteUsage()) {g=true}) {constMax=0.01}
  NamedRerouteDeclaration(None=LinearInterpolate(A=Constant() {r=0.2}, B=Multiply(A=@n47, B=@n48), Alpha=NamedRerouteDeclaration(None=@n49) {name=TopSeamMask, nodeColor=<rgba 1 0 0.8438 1>})) {name=Liquid Interior Pattern, nodeColor=<rgba 0.875 1 0 1>}
  n50 = ScalarParameter() {defaultValue=0.25, parameterName=Param_17}
  n51 = NamedRerouteUsage()
  n52 = Add(A=NamedRerouteUsage(), B=Constant() {r=-1})
  n53 = Constant() {r=-0.3}
  n54 = LinearInterpolate(
   A=VectorParameter() {parameterName=Liquid_Color1, defaultValue=<rgba 0.4948 0.043 0.0387 1>}.RGB
   B=VectorParameter() {parameterName=Liquid_Color2, defaultValue=<rgba 4 1.8856 0.0437 1>}.RGB
   Alpha=LinearInterpolate(
    A=SmoothStep(Min=@n50, Value=@n51)
    B=@n51
    Alpha=MaterialFunctionCall(In=@n52, Contrast=@n53) {materialFunction=&CheapContrast}.Result
   )
  )
  n55 = LinearInterpolate(
   A=@n54
   B=Multiply(A=@n54, B=VectorParameter() {parameterName="Liquid_Color3(sun)", defaultValue=<rgba 10 3.1252 5.1116 1>}.RGB)
   Alpha=Add(A=NamedRerouteUsage(), B=Constant())
  )
  n56 = Add(A=Multiply(A=NamedRerouteUsage(), B=ScalarParameter() {defaultValue=0.05, parameterName=Param_1}), B=NamedRerouteUsage())
  n57 = ScalarParameter() {defaultValue=0.32, parameterName=Param_5}
  n58 = OneMinus(None=@n56)
  n59 = ScalarParameter() {parameterName=Param_7}
  n60 = Min(
   A=MaterialFunctionCall(In=Add(A=@n56, B=@n57), Contrast=ScalarParameter() {defaultValue=10, parameterName=Param_6}) {materialFunction=&CheapContrast}.Result
   B=MaterialFunctionCall(In=Add(A=@n58, B=@n59), Contrast=ScalarParameter() {defaultValue=0.5, parameterName=Param_8}) {materialFunction=&CheapContrast}.Result
  )
  n61 = LinearInterpolate(
   A=VectorParameter() {parameterName=Param_1, defaultValue=<rgba 0.5 0.0649 0.0425 1>}.RGB
   B=VectorParameter() {defaultValue=<rgba 1 0.2462 0 1>}.RGB
   Alpha=Multiply(A=@n60, B=@n60)
  )
  n62 = NamedRerouteUsage()
  NamedRerouteDeclaration(None=Multiply(A=LinearInterpolate(A=@n55, B=@n61, Alpha=@n62), B=EyeAdaptationInverse().EyeAdaptationInverse)) {name=Emissive, nodeColor=<rgba 1 0 0.5391 1>}
  n63 = Add(A=NamedRerouteUsage(), B=Multiply(A=NamedRerouteUsage(), B=ScalarParameter() {defaultValue=-0.7, parameterName=Liquid_Distorion_BumpO}))
  n64 = AppendVector(A=Multiply(A=Time(), B=ScalarParameter() {defaultValue=-0.01, parameterName=Liquid_Distortion_Scroll}) {constB=0}, B=Constant())
  n65 = TextureSample(
   UVs=Multiply(A=Add(A=@n63, B=@n64), B=ScalarParameter() {defaultValue=1, parameterName=Liquid_Distortion_Tile})
   "DDX(UVs)"=NamedRerouteUsage()
   "DDY(UVs)"=NamedRerouteUsage()
  ) {mipValueMode=TMVM_Derivative, samplerType=SAMPLERTYPE_Grayscale, texture=/Game/Global/Textures/Default/Noise/LowResBlurredNoiseNormalized}
  NamedRerouteDeclaration(None=@n65.R) {name=Distortion Texture, nodeColor=<rgba 1 0.4219 0 1>}
  n66 = NamedRerouteUsage()
  n67 = ScalarParameter() {defaultValue=0.3, parameterName=Membrane_Thickness}
  NamedRerouteDeclaration(None=MaterialFunctionCall(In=Add(A=@n66, B=@n67), Contrast=ScalarParameter() {defaultValue=4, parameterName=Param_2}) {materialFunction=&CheapContrast}.Result) {name=M_MembraneBase, nodeColor=<rgba 0.4766 0 1 1>}
 ]}
