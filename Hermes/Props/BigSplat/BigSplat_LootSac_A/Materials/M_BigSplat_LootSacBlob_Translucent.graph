Material /Hermes/Props/BigSplat/BigSplat_LootSac_A/Materials/M_BigSplat_LootSacBlob_Translucent
graph {
 properties {
  bAllowFrontLayerTranslucency=false
  bAutomaticallySetUsageInEditor=false
  bUseMaterialAttributes=true
  blendMode=BLEND_Translucent
  translucencyLightingMode=TLM_VolumetricDirectional
  usedShadingModels=Default Lit
 }
 nodes [
  n1 = SetMaterialAttributes!(
   MaterialAttributes
   Base Color=NamedRerouteUsage()
   Emissive Color=NamedRerouteUsage()
   Specular=ScalarParameter() {defaultValue=1, parameterName=Specular}
   Roughness=ScalarParameter() {parameterName=Roughness}
   Metallic=ScalarParameter() {defaultValue=1, parameterName=Metallic}
   Opacity=NamedRerouteUsage()
   Normal
   Refraction
   World Position Offset=MaterialFunctionCall!(InvertNormal) {materialFunction=&/MF_BigSplat_LootSackBLob_WPO}.WPO
  ) {attributeSetTypes [
    #1
    #2
    #3
    #4
    #5
    #6
    #7
    #8
    #9
   ]}
  n2 = MaterialFunctionCall!(
   In=MaterialFunctionCall(Material_In=@n1) {materialFunction=&MF_FN_DestructibleDamage}.Material_Out
   UseSphericalMappingOverride
  ) {materialFunction=/Game/Packages/Fortress_SharedMaterials/Base_Material_Functions/MF_CurieFallbackEffect}
  Composite!(InputPin=@n2.Result) {subgraphName=Mobile Opt}
  Constant3Vector() {constant=<rgba 0.5636 0.2378 1 1>}
  Constant3Vector()
  Constant()
  n3 = NamedRerouteUsage()
  Desaturation(None=@n3, Fraction=ScalarParameter() {defaultValue=1, parameterName=Param_9})
  MaterialExpressionReroute_2 = Reroute(None=@n2.Result)
  n4 = GetMaterialAttributes(None=@MaterialExpressionReroute_2.InputPin)
  n5 = Constant() {r=0.5}
  n6 = Constant() {r=0.8}
  n7 = ShadingPathSwitch(
   Default=@MaterialExpressionReroute_2.InputPin
   Mobile=QualitySwitch(
    Default=@MaterialExpressionReroute_2.InputPin
    Low=SetMaterialAttributes!(MaterialAttributes=@n4.MaterialAttributes, Specular=@n5, Roughness=@n6) {attributeSetTypes [#3, #4]}
    High=@MaterialExpressionReroute_2.InputPin
    Medium=@MaterialExpressionReroute_2.InputPin
    Epic=@MaterialExpressionReroute_2.InputPin
   )
  )
  MaterialExpressionReroute_3 = Reroute(None=@n7)
  MaterialFunctionCall(MA In=@MaterialExpressionReroute_3.Output) {materialFunction=&MF_CreativeMoveTool}
  MaterialFunctionCall(
   Color_Morning=LinearInterpolate(
    A=VectorParameter() {parameterName=Liquid_Color_Morning1, defaultValue=<rgba 0.0411 0.092 0 1>}.RGB
    B=VectorParameter() {parameterName=Liquid_Color_Morning2, defaultValue=<rgba 1 0.8304 0 1>}.RGB
    Alpha=NamedRerouteUsage()
   )
   Color_Day=LinearInterpolate(
    A=VectorParameter() {parameterName=Liquid_Color_Day1, defaultValue=<rgba 0.3089 0.4149 0.0425 1>}.RGB
    B=VectorParameter() {parameterName=Liquid_Color_Day2, defaultValue=<rgba 0.9402 1 0.1878 1>}.RGB
    Alpha=NamedRerouteUsage()
   )
   Color_Evening=LinearInterpolate(
    A=VectorParameter() {parameterName=Liquid_Color_Evening1, defaultValue=<rgba 0.0253 0.0887 0.0266 1>}.RGB
    B=VectorParameter() {parameterName=Liquid_Color_Evening2, defaultValue=<rgba 0.8439 1 0.386 1>}.RGB
    Alpha=NamedRerouteUsage()
   )
   Color_Night=LinearInterpolate(
    A=VectorParameter() {parameterName=Liquid_Color_Night1, defaultValue=<rgba 0.004 0.0764 0.039 1>}.RGB
    B=VectorParameter() {parameterName=Liquid_Color_Night2, defaultValue=<rgba 0.1816 0.934 0.8786 1>}.RGB
    Alpha=NamedRerouteUsage()
   )
  ) {materialFunction=&Color_Time_of_Day}
  n8 = SceneDepth()
  n9 = PixelDepth()
  n10 = MaterialFunctionCall(Color=ObjectBounds()) {materialFunction=&SplitComponents}
  NamedRerouteDeclaration(None=Divide(A=Subtract(A=@n8.R, B=@n9.R), B=Max(A=@n10.R, B=@n10.R))) {name=M_Depht, nodeColor=<rgba 0 1 0.2734 1>}
  n11 = CameraVectorWS()
  n12 = VertexNormalWS()
  n13 = Constant() {r=0.5}
  n14 = SkyLightEnvMapSample(
   Direction=Add(A=Multiply(A=@n11) {constB=-1}, B=Multiply(A=@n12, B=@n13))
   Roughness=ScalarParameter() {group=CubeMap, parameterName=SkylightEnvMapRoughness}
  )
  n15 = LinearInterpolate(A=ScalarParameter() {parameterName=Param_11}, B=ScalarParameter() {defaultValue=1, parameterName=Param_10}, Alpha=NamedRerouteUsage())
  n16 = VectorParameter() {parameterName=Cubemap Color Lerp, defaultValue=<rgba 0.255 0.1068 0.0341 1>}
  n17 = ScalarParameter() {defaultValue=0.4, desc=0.03, parameterName=BG_CubemapMult_SidesAir}
  n18 = NamedRerouteUsage()
  NamedRerouteDeclaration(None=Multiply(A=Desaturation(None=@n14, Fraction=@n15), B=LinearInterpolate(A=@n16.RGB, B=@n17, Alpha=@n18))) {name=BaseColor, nodeColor=<rgba 1 0.3516 0 1>}
  n19 = ScalarParameter() {defaultValue=1, group=Roughness, parameterName=Rough_RepeatU}
  n20 = ScalarParameter() {defaultValue=1, group=Roughness, parameterName=Rough_RepeatV}
  n21 = TextureSampleParameter2D(UVs=Multiply(A=TextureCoordinate(), B=AppendVector(A=@n19, B=@n20))) {
   group=Grime
   parameterName=TextureMask
   samplerType=SAMPLERTYPE_Masks
   texture=/Game/Global/Materials/EnvironmentGlass/T_GlassTestMask_Indestruct
   channelNames {a="", b=Mask, g=Roughness, r=Grime}
  }
  NamedRerouteDeclaration(None=@n21.Roughness) {name=RoughMask, nodeColor=<rgba 0.602 0.7917 0.7205 1>}
  n22 = ScalarParameter() {
   defaultValue=1.5
   desc="Tighten/loosen the fresnel  blend.  Fresnel allows for different values of opacity and roughness based on the angle of the surface facing the camera. (Glass is shinier when looking at it from the edge)"
   parameterName=Fresnel Exponent
  }
  NamedRerouteDeclaration(None=Fresnel(ExponentIn=@n22) {baseReflectFraction=0}) {name=FresnelMask, nodeColor=<rgba 1 0.8814 0 1>}
  n23 = SkyLightEnvMapSample(Direction=ReflectionVectorWS(), Roughness=ScalarParameter() {group=CubeMap, parameterName=SkylightEnvMapRoughness})
  n24 = ScalarParameter() {
   defaultValue=0.5
   desc="A multiplier on the cubemap to blend it in and out.  Higher values show it more."
   group=CubeMap
   parameterName=CubeMap_Mult
  }
  n25 = ScalarParameter() {
   defaultValue=0.8
   desc=When the angle of the surface faces more away from the camera this will be the value used for Cubemap mult.
   group=CubeMap
   parameterName=CubeMap_Fresnel
  }
  n26 = NamedRerouteUsage()
  n27 = ScalarParameter() {defaultValue=0.5, group=Roughness, parameterName=RoughMask_Value_MobileForward}
  n28 = ShadingPathSwitch(Default=ScalarParameter() {defaultValue=0.5, group=Roughness, parameterName=RoughMask_Value}, Forward=@n27, Mobile=@n27)
  n29 = LinearInterpolate(A=LinearInterpolate(A=@n24, B=@n25, Alpha=NamedRerouteUsage()), Alpha=Saturate(None=Multiply(A=@n26, B=@n28))) {constB=0}
  NamedRerouteDeclaration(None=Multiply(A=Multiply(A=@n29, B=@n23), B=VectorParameter() {group=CubeMap, parameterName=CubemapColorScale, defaultValue=<rgba 1 1 1 0>}.RGB)) {name=CubeMapColor_Reflex, nodeColor=<rgba 0.2422 0 1 1>}
  n30 = Subtract(A=WorldPosition() {worldPositionShaderOffset=WPT_ExcludeAllShaderOffsets}.XYZ, B=ObjectPositionWS())
  n31 = ObjectLocalBounds()
  n32 = Divide(A=Add(A=@n30, B=@n31.Half Extents) {constB=0.5}, B=@n31.Extents)
  n33 = AppendVector(
   A=MaterialFunctionCall(Vector or UVs=ComponentMask(None=@n30) {g=true, r=true}) {materialFunction=&VectorToRadialValue}.Vector Converted to Angle
   B=NamedRerouteDeclaration(None=ComponentMask(None=@n32) {b=true}) {name=Local H, nodeColor=<rgba 0.5469 1 0 1>}
  )
  NamedRerouteDeclaration(None=Add(A=Multiply(A=@n33) {constB=-1})) {name=UV_Cylindrical, nodeColor=<rgba 1 0.8438 0 1>}
  n34 = TextureCoordinate()
  NamedRerouteDeclaration(None=DDX(Value=@n34)) {name=DDX, nodeColor=<rgba 0 0.2734 1 1>}
  NamedRerouteDeclaration(None=DDY(Value=@n34)) {name=DDY, nodeColor=<rgba 0 0.2734 1 1>}
  n35 = Multiply(A=DotProduct(A=SkyAtmosphereLightDirection(), B=CameraVectorWS())) {constB=-1}
  n36 = ScalarParameter() {defaultValue=-0.7, parameterName=SunAuraSize}
  n37 = Saturate(None=Multiply(A=Add(A=@n35, B=@n36), B=Constant() {r=3}))
  NamedRerouteDeclaration(None=Multiply(A=@n37, B=@n37)) {name=SunBlob, nodeColor=<rgba 1 0 0.8203 1>}
  n38 = Time()
  n39 = ScalarParameter() {defaultValue=-0.005, parameterName=Liquid1_ScrollX}
  n40 = ScalarParameter() {defaultValue=0.01, parameterName=Liquid1_ScrollY}
  n41 = Add(A=MaterialFunctionCall!(Position=ObjectPositionWS(), UVRadialScale) {materialFunction=/Game/Weapons/WeaponSkins/MaterialFunction/ObjectStabilizedScreenSpaceUVs}.StabilizedUVs) {constB=-0.5}
  n42 = NamedRerouteDeclaration(None=ComponentMask(None=Multiply(A=@n41) {constB=-1}) {g=true, r=true}) {name=BumpOffsetUV, nodeColor=<rgba 0.125 0 1 1>}
  n43 = Add(A=Constant() {r=-0.3}, B=Multiply(A=NamedRerouteUsage(), B=ScalarParameter() {defaultValue=0.1, parameterName=Liquid_Distortion_1}))
  n44 = Add(A=Add(A=NamedRerouteUsage(), B=Multiply(A=@n42, B=@n43)), B=AppendVector(A=Multiply(A=@n38, B=@n39), B=Multiply(A=@n38, B=@n40)))
  n45 = Multiply(A=@n44, B=ScalarParameter() {defaultValue=1, desc=3, parameterName=Liquid1_Tile})
  n46 = NamedRerouteUsage()
  n47 = Add(
   A=ScalarParameter() {defaultValue=-0.7, parameterName=Liquid_Distorion_BumpO}
   B=Multiply(A=NamedRerouteUsage(), B=ScalarParameter() {defaultValue=0.25, parameterName=Liquid_Distortion_2})
  )
  n48 = Time()
  n49 = ScalarParameter() {defaultValue=-0.003, parameterName=Liquid2_ScrollX}
  n50 = ScalarParameter() {defaultValue=0.01, parameterName=Liquid2_ScrollY}
  n51 = Add(A=Add(A=NamedRerouteUsage(), B=Multiply(A=@n46, B=@n47)), B=AppendVector(A=Multiply(A=@n48, B=@n49), B=Multiply(A=@n48, B=@n50)))
  n52 = ScalarParameter() {defaultValue=2, desc=6, parameterName=Liquid2_Tile}
  n53 = Max(
   A=TextureSample!(UVs=Frac(None=@n45), Tex, "DDX(UVs)"=NamedRerouteUsage(), "DDY(UVs)"=NamedRerouteUsage(), Apply View MipBias) {mipValueMode=TMVM_Derivative, texture=/Hermes/Props/CozyTea/CozyTea_PickleSign_A/Textures/T_CozyTea_PickleSign_M}.RGB
   B=TextureSample(UVs=Multiply(A=@n51, B=@n52), "DDX(UVs)"=NamedRerouteUsage(), "DDY(UVs)"=NamedRerouteUsage()) {mipValueMode=TMVM_Derivative, texture=&T_CozyTea_PickleSign_M}.RGB
  )
  n54 = ComponentMask(None=@n53) {r=true}
  n55 = Constant() {r=1}
  n56 = ComponentMask(None=@n53) {g=true}
  n57 = Add(A=@n56, B=Constant() {r=1})
  n58 = Constant() {r=0.2}
  n59 = Multiply(A=@n51, B=ScalarParameter() {defaultValue=1, parameterName=Liquid2_TileMask})
  n60 = NamedRerouteUsage()
  n61 = NamedRerouteUsage()
  n62 = TextureSample(
   UVs=Multiply(A=@n44, B=ScalarParameter() {defaultValue=2, parameterName=Liquid1_TileMask})
   "DDX(UVs)"=NamedRerouteUsage()
   "DDY(UVs)"=NamedRerouteUsage()
  ) {mipValueMode=TMVM_Derivative, texture=&T_CozyTea_PickleSign_M}
  n63 = Add(
   A=Multiply(
    A=TextureSample(UVs=@n59, "DDX(UVs)"=@n60, "DDY(UVs)"=@n61) {mipValueMode=TMVM_Derivative, texture=&T_CozyTea_PickleSign_M}.B
    B=Constant() {r=1}
   ) {constB=0.5}
   B=Multiply(A=NamedRerouteDeclaration(None=@n62.B) {name=Noise_2, nodeColor=<rgba 0.1953 1 0 1>}, B=Constant() {r=0.8}) {constB=0.2}
  )
  n64 = ScalarParameter() {defaultValue=-0.3, desc=-.3, parameterName=Liquid_Viscosity_Bias}
  n65 = LinearInterpolate(
   A=Max(A=Multiply(A=@n54, B=@n55), B=Multiply(A=@n57, B=@n58))
   Alpha=MaterialFunctionCall(In=Add(A=@n63, B=@n64), Contrast=ScalarParameter() {defaultValue=-0.35, desc=-.2, parameterName=Liquid_Visocity_Contrast}) {materialFunction=&CheapContrast}.Result
  ) {constB=0}
  n66 = Saturate(None=LinearInterpolate(A=Constant() {r=-0.2}, B=Constant() {r=2}, Alpha=NamedRerouteUsage()))
  n67 = SmoothStep(Min=Constant() {r=0.02}, Max=Constant() {r=0.08}, Value=ComponentMask(None=NamedRerouteUsage()) {g=true}) {constMax=0.01}
  NamedRerouteDeclaration(None=LinearInterpolate(A=Constant() {r=0.2}, B=Multiply(A=@n65, B=@n66), Alpha=NamedRerouteDeclaration(None=@n67) {name=TopSeamMask, nodeColor=<rgba 1 0 0.8437 1>})) {name=Liquid Interior Pattern, nodeColor=<rgba 0.875 1 0 1>}
  NamedRerouteDeclaration(None=@n56) {name=M_SurfaceDistortion, nodeColor=<rgba 0.3594 1 0 1>}
  n68 = LinearInterpolate(
   A=VectorParameter() {parameterName=Liquid_Color1, defaultValue=<rgba 0.2205 0.0002 0 1>}.RGB
   B=VectorParameter() {parameterName=Liquid_Color2, defaultValue=<rgba 4 1.3322 0 1>}.RGB
   Alpha=NamedRerouteUsage()
  )
  n69 = LinearInterpolate(A=@n68, B=Multiply(A=@n68, B=Constant() {r=0.5}), Alpha=Add(A=NamedRerouteUsage(), B=Constant()))
  n70 = LinearInterpolate(
   A=@n69
   B=Multiply(A=@n69, B=VectorParameter() {parameterName="Liquid_Color3(sun)", defaultValue=<rgba 10 3.1252 5.1116 1>}.RGB)
   Alpha=Add(A=NamedRerouteUsage(), B=Constant())
  )
  n71 = Add(A=Multiply(A=NamedRerouteUsage(), B=ScalarParameter() {defaultValue=0.05, parameterName=Param_1}), B=NamedRerouteUsage())
  n72 = ScalarParameter() {defaultValue=0.32, parameterName=Param_5}
  n73 = OneMinus(None=@n71)
  n74 = ScalarParameter() {parameterName=Param_7}
  n75 = Min(
   A=MaterialFunctionCall(In=Add(A=@n71, B=@n72), Contrast=ScalarParameter() {defaultValue=10, parameterName=Param_6}) {materialFunction=&CheapContrast}.Result
   B=MaterialFunctionCall(In=Add(A=@n73, B=@n74), Contrast=ScalarParameter() {defaultValue=0.5, parameterName=Param_8}) {materialFunction=&CheapContrast}.Result
  )
  n76 = LinearInterpolate(
   A=VectorParameter() {parameterName=Param_1, defaultValue=<rgba 1 0.0488 0 1>}.RGB
   B=VectorParameter() {defaultValue=<rgba 2 0.4923 0 1>}.RGB
   Alpha=Multiply(A=@n75, B=@n75)
  )
  n77 = LinearInterpolate(A=@n76, B=Multiply(A=@n76, B=@n3), Alpha=ScalarParameter() {defaultValue=0.9, parameterName=Membrane_CubemapReflex}) {constB=0}
  n78 = NamedRerouteUsage()
  NamedRerouteDeclaration(None=Multiply(A=LinearInterpolate(A=@n70, B=@n77, Alpha=@n78), B=EyeAdaptationInverse().EyeAdaptationInverse)) {name=Emissive, nodeColor=<rgba 1 0 0.5391 1>}
  n79 = Add(A=NamedRerouteUsage(), B=Multiply(A=NamedRerouteUsage(), B=ScalarParameter() {defaultValue=-0.7, parameterName=Liquid_Distorion_BumpO}))
  n80 = AppendVector(A=Multiply(A=Time(), B=ScalarParameter() {defaultValue=-0.01, parameterName=Liquid_Distortion_Scroll}) {constB=0}, B=Constant())
  n81 = TextureSample(
   UVs=Multiply(A=Add(A=@n79, B=@n80), B=ScalarParameter() {defaultValue=1, parameterName=Liquid_Distortion_Tile})
   "DDX(UVs)"=NamedRerouteUsage()
   "DDY(UVs)"=NamedRerouteUsage()
  ) {mipValueMode=TMVM_Derivative, samplerType=SAMPLERTYPE_Grayscale, texture=/Game/Global/Textures/Default/Noise/LowResBlurredNoiseNormalized}
  NamedRerouteDeclaration(None=@n81.R) {name=Distortion Texture, nodeColor=<rgba 1 0.4219 0 1>}
  n82 = Constant() {r=1}
  n83 = MaterialFunctionCall(
   In=Add(A=NamedRerouteUsage(), B=ScalarParameter() {defaultValue=-0.22, desc=-0.45})
   Contrast=ScalarParameter() {defaultValue=1, parameterName=Param_3}
  ) {materialFunction=&CheapContrast}
  n84 = NamedRerouteDeclaration(None=StaticBoolParameter() {parameterName=Use_Silhouette}) {name=Use_Silhouette, nodeColor=<rgba 1 0.0938 0 1>}
  n85 = Add(A=NamedRerouteUsage(), B=Constant() {r=0.05})
  n86 = Constant() {r=0.23}
  n87 = SmoothStep(
   Min=Constant()
   Max=Constant() {r=0.2}
   Value=Add(
    A=MaterialFunctionCall(In=@n85, Contrast=@n86) {materialFunction=&CheapContrast}.Result
    B=NamedRerouteUsage()
   )
  ) {constMax=0.3}
  n88 = Constant() {r=0.8}
  n89 = ScalarParameter() {defaultValue=0.35, parameterName=Membrane_Thickness}
  n90 = NamedRerouteDeclaration(None=Add(A=NamedRerouteUsage(), B=@n89)) {name=M_MembraneMask, nodeColor=<rgba 0 1 0.0156 1>}
  n91 = ScalarParameter() {defaultValue=4, parameterName=Param_2}
  n92 = LinearInterpolate(
   A=Max(A=StaticSwitch(True=@n82, False=@n83.Result, Value=@n84), B=Multiply(A=@n87, B=@n88))
   B=ScalarParameter() {defaultValue=0.8, parameterName=Membrane_Opacity}
   Alpha=NamedRerouteDeclaration(None=MaterialFunctionCall(In=@n90, Contrast=@n91) {materialFunction=&CheapContrast}.Result) {name=M_MembraneBase, nodeColor=<rgba 0.4766 0 1 1>}
  )
  n93 = NamedRerouteUsage()
  n94 = ScalarParameter() {defaultValue=0.2, group=Opacity, parameterName=Opacity_Fresnel_Center}
  n95 = ScalarParameter() {defaultValue=1, group=Opacity, parameterName=Opacity_Fresnel_Borders}
  n96 = Saturate(None=NamedRerouteUsage())
  NamedRerouteDeclaration(None=Max(A=Max(A=@n92, B=@n93), B=LinearInterpolate(A=@n94, B=@n95, Alpha=@n96))) {name=Opacity, nodeColor=<rgba 0.125 0 1 1>}
  NamedRerouteDeclaration(None=Multiply(A=NamedRerouteUsage(), B=ScalarParameter() {parameterName=Param_4})) {name=WPO, nodeColor=<rgba 0 1 0.8828 1>}
  NamedRerouteDeclaration(None=@n89) {name=S_MembraneThickness, nodeColor=<rgba 0 0.7187 1 1>}
  PinBase() {reroutePins [InputPin {expression=&self:MaterialExpressionReroute_2}]}
  PinBase!(Output=@n7) {reroutePins [Output {expression=&self:MaterialExpressionReroute_3}]}
  n97 = MaterialFunctionCall(
   Color_Morning=VectorParameter() {parameterName="Liquid_Color_Morning3(sun)", defaultValue=<rgba 10 3.1252 5.1116 1>}.RGB
   Color_Day=VectorParameter() {parameterName="Liquid_Color_Day3(sun)", defaultValue=<rgba 2.2057 3 1.5833 1>}.RGB
   Color_Evening=VectorParameter() {parameterName="Liquid_Color_Evening(sun)", defaultValue=<rgba 6 1.5635 1.3047 1>}.RGB
   Color_Night=VectorParameter() {parameterName="Liquid_Color_Night(sun)", defaultValue=<rgba 4.5 5.598 6 1>}.RGB
  ) {materialFunction=&Color_Time_of_Day}
  ShadingPathSwitch(Default=@n97.Result, Mobile=Multiply(A=@n97.Result) {constB=0.4})
  VectorParameter() {parameterName=Membrane_Color, defaultValue=<rgba 2 0.5629 0 0.7>}
  VectorParameter() {parameterName=Param_2, defaultValue=<rgba 0 0 0 1>}
 ]
}
