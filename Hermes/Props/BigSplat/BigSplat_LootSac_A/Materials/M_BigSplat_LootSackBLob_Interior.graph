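# Material for the interior blob of the BigSplat loot sack prop, built on material attributes with Default Lit shading.
# Base color blends two color parameters by an offset, contrast-adjusted normal-to-camera dot product, while a shared blob function drives world position offset with inverted normals.
# The result chains through destructible damage, Curie fallback and Creative move tool functions.
# A Mobile Opt subgraph overrides Specular and Roughness with constants on low-quality mobile.
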Material /Hermes/Props/BigSplat/BigSplat_LootSac_A/Materials/M_BigSplat_LootSackBLob_Interior
graph {
 properties {bAutomaticallySetUsageInEditor=false, bUseMaterialAttributes=true, usedShadingModels=Default Lit}
 nodes [
  n1 = Add(A=DotProduct(A=VertexNormalWS(), B=CameraVectorWS()), B=ScalarParameter() {defaultValue=-0.15})
  n2 = ScalarParameter() {defaultValue=1, parameterName=Param_1}
  n3 = SetMaterialAttributes!(
   MaterialAttributes
   Base Color=LinearInterpolate(
    A=VectorParameter() {parameterName=BaseColor_2, defaultValue=<rgba 1 0.0892 0 1>}.RGB
    B=VectorParameter() {parameterName=BaseColor_1, defaultValue=<rgba 1 0.6261 0.0851 1>}.RGB
    Alpha=MaterialFunctionCall(In=@n1, Contrast=@n2) {materialFunction=&CheapContrast}.Result
   )
   Metallic
   Specular
   Roughness
   World Position Offset=MaterialFunctionCall!(InvertNormal=StaticBool() {value=true}) {materialFunction=&/MF_BigSplat_LootSackBLob_WPO}.WPO
  ) {attributeSetTypes [
    #1
    #2
    #3
    #4
    #5
   ]}
  n4 = MaterialFunctionCall!(
   In=MaterialFunctionCall(Material_In=@n3) {materialFunction=&MF_FN_DestructibleDamage}.Material_Out
   UseSphericalMappingOverride
  ) {materialFunction=/Game/Packages/Fortress_SharedMaterials/Base_Material_Functions/MF_CurieFallbackEffect}
  Composite!(InputPin=@n4.Result) {subgraphName=Mobile Opt}
  MaterialExpressionReroute_0 = Reroute(None=@n4.Result)
  n5 = GetMaterialAttributes(None=@MaterialExpressionReroute_0.InputPin)
  n6 = Constant() {r=0.5}
  n7 = Constant() {r=0.8}
  n8 = ShadingPathSwitch(
   Default=@MaterialExpressionReroute_0.InputPin
   Mobile=QualitySwitch(
    Default=@MaterialExpressionReroute_0.InputPin
    Low=SetMaterialAttributes!(MaterialAttributes=@n5.MaterialAttributes, Specular=@n6, Roughness=@n7) {attributeSetTypes [#3, #4]}
    High=@MaterialExpressionReroute_0.InputPin
    Medium=@MaterialExpressionReroute_0.InputPin
    Epic=@MaterialExpressionReroute_0.InputPin
   )
  )
  MaterialExpressionReroute_1 = Reroute(None=@n8)
  MaterialFunctionCall(MA In=@MaterialExpressionReroute_1.Output) {materialFunction=&MF_CreativeMoveTool}
  PinBase!(Output=@n8) {reroutePins [Output {expression=&self:MaterialExpressionReroute_1}]}
  PinBase!() {reroutePins [InputPin {expression=&self:MaterialExpressionReroute_0}]}
 ]
}
